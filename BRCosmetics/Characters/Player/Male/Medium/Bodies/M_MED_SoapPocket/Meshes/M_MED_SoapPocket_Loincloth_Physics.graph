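# Physics asset for the loincloth on the medium male SoapPocket player body, with capsule bodies for the cloth chains, pelvis, thighs and calves.
# Left and right five-bone dyn_loincloth chains use limited swing and tight twist, with their first bones angularly locked to base bones under the pelvis.
# An angularly free constraint joins the two chain tips.
# Leg and base-to-pelvis joints rotate freely; calf_r links directly to pelvis, unlike calf_l to thigh_l.
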PhysicsAsset /BRCosmetics/Characters/Player/Male/Medium/Bodies/M_MED_SoapPocket/Meshes/M_MED_SoapPocket_Loincloth_Physics
bodies {
 base_loincloth_01_l [{shapeType=Capsule, length=0.505, radius=0.505, shapeName=base_loincloth_01_l_capsule, rotation=<rot 0 -90 0>}]
 base_loincloth_01_r [{shapeType=Capsule, length=0.505, radius=0.505, shapeName=base_loincloth_01_r_capsule, rotation=<rot 0 -90 0>}]
 calf_l [{
   shapeType=Capsule
   length=10.7242
   radius=7.123
   shapeName=calf_l_capsule
   center=<1.3326 0 0.0861>
   rotation=<rot 86.3051 179.9996 179.9998>
  }]
 calf_r [{shapeType=Capsule, length=11.6166, radius=6.2049, shapeName=calf_r_capsule, rotation=<rot 90 0 0>}]
 dyn_loincloth_01_l [{shapeType=Capsule, length=3, radius=4, shapeName=dyn_loincloth_01_l_capsule, center=<6 0 0>, rotation=<rot -90 0 0>}]
 dyn_loincloth_01_r [{shapeType=Capsule, length=3, radius=4, shapeName=dyn_loincloth_01_r_capsule, center=<6 0 0>, rotation=<rot -90 0 0>}]
 dyn_loincloth_02_l [{
   shapeType=Capsule
   length=5.6121
   radius=4.3879
   shapeName=dyn_loincloth_02_l_capsule
   center=<5.0411 -0.0002 0.0881>
   rotation=<rot 88.9985 179.9999 180>
  }]
 dyn_loincloth_02_r [{
   shapeType=Capsule
   length=6.451
   radius=3.9235
   shapeName=dyn_loincloth_02_l_capsule
   center=<5.0411 -0.0002 0.0881>
   rotation=<rot 88.9985 179.9999 180>
  }]
 dyn_loincloth_03_l [{
   shapeType=Capsule
   length=6.4023
   radius=3.5977
   shapeName=dyn_loincloth_03_l_capsule
   center=<5.0411 -0.0002 0.0877>
   rotation=<rot 89.0034 179.9999 180>
  }]
 dyn_loincloth_03_r [{
   shapeType=Capsule
   length=5.7369
   radius=4.2631
   shapeName=dyn_loincloth_03_r_capsule
   center=<5.0411 -0.0002 0.0877>
   rotation=<rot 89.0034 179.9999 180>
  }]
 dyn_loincloth_04_l [{
   shapeType=Capsule
   length=6.5284
   radius=3.4716
   shapeName=dyn_loincloth_04_l_capsule
   center=<4.7374 -0.0001 0.0824>
   rotation=<rot 89.0034 179.9999 -180>
  }]
 dyn_loincloth_04_r [{
   shapeType=Capsule
   length=6.4216
   radius=3.5784
   shapeName=dyn_loincloth_04_r_capsule
   center=<4.7374 -0.0001 0.0828>
   rotation=<rot 88.9985 179.9999 -180>
  }]
 dyn_loincloth_05_l [{
   shapeType=Capsule
   length=6.43
   radius=3.57
   shapeName=dyn_loincloth_05_l_capsule
   center=<4.7374 -0.0001 0.0824>
   rotation=<rot 89.0034 179.9999 -180>
  }]
 dyn_loincloth_05_r [{
   shapeType=Capsule
   length=7.1131
   radius=2.8869
   shapeName=dyn_loincloth_05_r_capsule
   center=<4.7374 -0.0001 0.0824>
   rotation=<rot 89.0034 179.9999 -180>
  }]
 pelvis [{shapeType=Capsule, length=14.5541, radius=14.9713, shapeName=pelvis_capsule, center=<-15.7787 -6.0847 0>, rotation=<rot 90 0 0>}]
 thigh_l [{shapeType=Capsule, length=24.5128, radius=8.5074, center=<-15.5961 0 -0.7>, rotation=<rot 90 0 0>}]
 thigh_r [{shapeType=Capsule, length=24.5128, radius=8.5074, center=<15.5961 0 1.318>, rotation=<rot 90 0 180>}]
}
constraints [
 {bone1Name=dyn_loincloth_02_r, bone2Name=dyn_loincloth_01_r, swing1LimitDegrees=25, swing2LimitDegrees=15, twistLimitDegrees=5}
 {bone1Name=dyn_loincloth_03_r, bone2Name=dyn_loincloth_02_r, swing1LimitDegrees=25, swing2LimitDegrees=5, twistLimitDegrees=5}
 {bone1Name=dyn_loincloth_04_r, bone2Name=dyn_loincloth_03_r, swing1LimitDegrees=25, swing2LimitDegrees=5, twistLimitDegrees=5}
 {bone1Name=dyn_loincloth_05_r, bone2Name=dyn_loincloth_04_r, swing1LimitDegrees=25, swing2LimitDegrees=5, twistLimitDegrees=5}
 {bone1Name=dyn_loincloth_02_l, bone2Name=dyn_loincloth_01_l, swing1LimitDegrees=25, swing2LimitDegrees=15, twistLimitDegrees=5}
 {bone1Name=dyn_loincloth_03_l, bone2Name=dyn_loincloth_02_l, swing1LimitDegrees=25, swing2LimitDegrees=5, twistLimitDegrees=5}
 {bone1Name=dyn_loincloth_04_l, bone2Name=dyn_loincloth_03_l, swing1LimitDegrees=25, swing2LimitDegrees=5, twistLimitDegrees=5}
 {bone1Name=dyn_loincloth_05_l, bone2Name=dyn_loincloth_04_l, swing1LimitDegrees=25, swing2LimitDegrees=5, twistLimitDegrees=5}
 {
  bone1Name=dyn_loincloth_05_r
  bone2Name=dyn_loincloth_05_l
  swing1LimitDegrees=45
  swing1Motion=Free
  swing2LimitDegrees=45
  swing2Motion=Free
  twistLimitDegrees=45
  twistMotion=Free
 }
 {
  bone1Name=calf_l
  bone2Name=thigh_l
  swing1LimitDegrees=45
  swing1Motion=Free
  swing2LimitDegrees=45
  swing2Motion=Free
  twistLimitDegrees=45
  twistMotion=Free
 }
 {
  bone1Name=dyn_loincloth_01_l
  bone2Name=base_loincloth_01_l
  swing1LimitDegrees=45
  swing1Motion=Locked
  swing2LimitDegrees=45
  swing2Motion=Locked
  twistLimitDegrees=45
  twistMotion=Locked
 }
 {
  bone1Name=dyn_loincloth_01_r
  bone2Name=base_loincloth_01_r
  swing1LimitDegrees=45
  swing1Motion=Locked
  swing2LimitDegrees=45
  swing2Motion=Locked
  twistLimitDegrees=45
  twistMotion=Locked
 }
 {
  bone1Name=thigh_l
  bone2Name=pelvis
  swing1LimitDegrees=45
  swing1Motion=Free
  swing2LimitDegrees=45
  swing2Motion=Free
  twistLimitDegrees=45
  twistMotion=Free
 }
 {
  bone1Name=thigh_r
  bone2Name=pelvis
  swing1LimitDegrees=45
  swing1Motion=Free
  swing2LimitDegrees=45
  swing2Motion=Free
  twistLimitDegrees=45
  twistMotion=Free
 }
 {
  bone1Name=calf_r
  bone2Name=pelvis
  swing1LimitDegrees=45
  swing1Motion=Free
  swing2LimitDegrees=45
  swing2Motion=Free
  twistLimitDegrees=45
  twistMotion=Free
 }
 {
  bone1Name=base_loincloth_01_r
  bone2Name=pelvis
  swing1LimitDegrees=45
  swing1Motion=Free
  swing2LimitDegrees=45
  swing2Motion=Free
  twistLimitDegrees=45
  twistMotion=Free
 }
 {
  bone1Name=base_loincloth_01_l
  bone2Name=pelvis
  swing1LimitDegrees=45
  swing1Motion=Free
  swing2LimitDegrees=45
  swing2Motion=Free
  twistLimitDegrees=45
  twistMotion=Free
 }
]
