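# Animation Blueprint for the M_MED_SoapPocket male body cosmetic, adding secondary loincloth motion to the pose copied from the parent mesh.
# A Control Rig, fed skydive and parachute states, produces a cached pose for two pelvis-space RigidBody branches.
# Standing adds loincloth bone offsets before and after simulation; crouching swaps to its own physics asset without offsets.
# EventGraph updates parachute alongside gravity, front-end and dynamics variables that AnimGraph never reads.
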AnimBlueprint /BRCosmetics/Characters/Player/Male/Medium/Bodies/M_MED_SoapPocket/Meshes/M_MED_SoapPocket_AnimBP
target_skeleton=&/M_MED_SoapPocket_Skeleton
graphs [
 AnimGraph {nodes [
   n1 = "Transform (Modify) Bone - Bone: base_loincloth_01_l"!(
    ComponentPose=Local To Component(LocalPose=Use cached pose 'CRCache_Input'())
    bAlphaBoolEnabled:bool=True
    Translation="-2,0,0"
    Rotation="0,30,0"
    Scale="1,1,1"
   ) -> (Pose) {
    alphaInputType=Bool
    rotationMode=BMM_Additive
    rotationSpace=BCS_BoneSpace
    scaleMode=BMM_Replace
    translationMode=BMM_Additive
    translationSpace=BCS_BoneSpace
    alphaBoolBlend {blendInTime=0.2, blendOption=Linear, blendOutTime=0.2}
    boneToModify {boneName=base_loincloth_01_l}
    rotation {yaw=30}
    translation {x=-2}
   }
   n2 = "Transform (Modify) Bone - Bone: base_loincloth_01_r"!(ComponentPose=@n1, bAlphaBoolEnabled:bool=True, Translation="-2,0,0", Rotation="0,50,0", Scale="1,1,1") -> (Pose) {
    alphaInputType=Bool
    rotationMode=BMM_Additive
    rotationSpace=BCS_BoneSpace
    scaleMode=BMM_Replace
    translationMode=BMM_Additive
    translationSpace=BCS_BoneSpace
    alphaBoolBlend {blendInTime=0.2, blendOption=Linear, blendOutTime=0.2}
    boneToModify {boneName=base_loincloth_01_r}
    rotation {yaw=50}
    translation {x=-2}
   }
   n3 = RigidBody!(ComponentPose=@n2, Alpha, OverrideWorldGravity:Vector="0,0,-980", ExternalForce) -> (Pose) {
    bOverrideWorldGravity=true
    overridePhysicsAsset=&/M_MED_SoapPocket_Loincloth_Physics
    simulationSpace=BaseBoneSpace
    alphaBoolBlend {blendOption=Linear}
    baseBoneRef {boneName=pelvis}
    componentAppliedLinearAccClamp=<3000 3000 3000>
    overrideWorldGravity {z=-980}
    simSpaceSettings {
     maxAngularAcceleration=1
     maxAngularVelocity=3
     maxLinearAcceleration=500
     maxLinearVelocity=100
     velocityScaleZ=0.4
     worldAlpha=1
     externalLinearDragV=<0.2 0.1 0.2>
    }
   }
   n4 = "Transform (Modify) Bone - Bone: base_loincloth_01_l"(ComponentPose=@n3, bAlphaBoolEnabled=NOT Boolean(A=Get bIsInFrontEnd()), Translation="0,0,0", Rotation="0.000,0.000,-10.000", Scale="1,1,1") {
    alphaInputType=Bool
    rotationMode=BMM_Additive
    rotationSpace=BCS_BoneSpace
    scaleMode=BMM_Replace
    alphaBoolBlend {blendOption=Linear}
    boneToModify {boneName=base_loincloth_01_l}
    rotation {roll=-10, yaw=0}
   }
   n5 = "Transform (Modify) Bone - Bone: base_loincloth_01_l"(ComponentPose=@n4, bAlphaBoolEnabled=True, Translation="0,0,0", Rotation="0,-10,0", Scale="1,1,1") {
    alphaInputType=Bool
    rotationMode=BMM_Additive
    rotationSpace=BCS_BoneSpace
    scaleMode=BMM_Replace
    translationMode=BMM_Additive
    translationSpace=BCS_BoneSpace
    alphaBoolBlend {blendInTime=0.1, blendOption=Linear, blendOutTime=0.1}
    boneToModify {boneName=base_loincloth_01_l}
    rotation {yaw=-10}
   }
   n6 = "Transform (Modify) Bone - Bone: base_loincloth_01_r"(ComponentPose=@n5, bAlphaBoolEnabled=True, Translation="0,0,0", Rotation="1,-5,0", Scale="1,1,1") {
    alphaInputType=Bool
    rotationMode=BMM_Additive
    rotationSpace=BCS_BoneSpace
    scaleMode=BMM_Replace
    translationMode=BMM_Additive
    translationSpace=BCS_BoneSpace
    alphaBoolBlend {blendInTime=0.1, blendOption=Linear, blendOutTime=0.1}
    boneToModify {boneName=base_loincloth_01_r}
    rotation {pitch=1, yaw=-5}
   }
   n7 = "Transform (Modify) Bone - Bone: base_loincloth_01_l"(ComponentPose=@n6, bAlphaBoolEnabled=True, Translation="0,0,0", Rotation="0,2,0", Scale="1,1,1") {
    alphaInputType=Bool
    rotationMode=BMM_Additive
    rotationSpace=BCS_BoneSpace
    scaleMode=BMM_Replace
    translationMode=BMM_Additive
    translationSpace=BCS_BoneSpace
    alphaBoolBlend {blendInTime=0.1, blendOption=Linear, blendOutTime=0.1}
    boneToModify {boneName=base_loincloth_01_l}
    rotation {yaw=2}
   }
   n8 = "Transform (Modify) Bone - Bone: base_loincloth_01_r"(ComponentPose=@n7, bAlphaBoolEnabled=True, Translation="0,0,1", Rotation="3,2,0", Scale="1,1,1") {
    alphaInputType=Bool
    rotationMode=BMM_Additive
    rotationSpace=BCS_BoneSpace
    scaleMode=BMM_Replace
    translationMode=BMM_Additive
    translationSpace=BCS_BoneSpace
    alphaBoolBlend {blendInTime=0.1, blendOption=Linear, blendOutTime=0.1}
    boneToModify {boneName=base_loincloth_01_r}
    rotation {pitch=3, yaw=2}
    translation {z=1}
   }
   n9 = RigidBody(ComponentPose=Local To Component(LocalPose=Use cached pose 'CRCache_Input'()), OverrideWorldGravity="0,0,-980") {
    bOverrideWorldGravity=true
    overridePhysicsAsset=&/M_MED_SoapPocket_Crouch_Physics
    simulationSpace=BaseBoneSpace
    alphaBoolBlend {blendOption=Linear}
    baseBoneRef {boneName=pelvis}
    componentAppliedLinearAccClamp=<3000 3000 3000>
    overrideWorldGravity {z=-980}
    simSpaceSettings {
     maxAngularAcceleration=1
     maxAngularVelocity=3
     maxLinearAcceleration=500
     maxLinearVelocity=100
     worldAlpha=1
     externalLinearDragV {x=0.2, z=0.2}
    }
   }
   Output Pose(Result=Blend Poses by bool(BlendPose_0=Component To Local(ComponentPose=@n9), BlendPose_1=Component To Local(ComponentPose=@n8), bActiveValue=Get bIsCrouching()) {blendType=Linear}) {layerGroup=None}
   n10 = Control Rig!(
    Source=Copy Pose From Mesh() {bUseAttachedParent=true}
    Alpha
    isParachuting:double=Get isParachuteArms() -> (isParachuteArms:double)
    isSkydive:bool=Get bIsSkyDiving()
   ) -> (Pose) {
    alphaBoolBlend {blendOption=Linear}
    controlRigAssetReference {blueprintRigClass=&/M_MED_SoapPocket_CtrlRig.M_MED_SoapPocket_CtrlRig_C}
   }
   Save cached pose 'CRCache_Input'(Pose=@n10) {cacheName=CRCache_Input}
  ]}
 EventGraph {nodes [
   Event BlueprintUpdateAnimation()
   > n1 = Gravity Override State Function DEPRECATED(CharacterPartAnimInstance=Self-Reference(), GameTimeInSeconds=GetGameTimeInSeconds(), GravityOverrideParams=Get Gravity Override Params())
   > Set GravityOverride(GravityOverride=@n1)
   > Set isParachuteArms(isParachuteArms:double="To Float (Boolean)"(InBool=Get bIsParachuteOpened())) -> (Output_Get:double)
   n2 = Get bIsPlayingEmote()
   > Set FrontEndNoEmote(FrontEndNoEmote:bool=AND Boolean(A=Get bIsInFrontEnd(), B=NOT Boolean(A=@n2))) -> (Output_Get:bool)
   > Set notDyn(notDyn:bool=NOT Boolean(A=Get bRigidBodyEnabled())) -> (Output_Get:bool)
  ]}
]
variables [
 FrontEndNoEmote {type=boolean}
 Gravity Override Params {type=GravityOverrideParamsStruct, default {dBNO {gravityOverride=<-100 0 0>}}}
 GravityOverride {type=Vector}
 isParachuteArms {type=number}
 notDyn {type=boolean}
]
